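# Physics asset for the Rust Lord Remix backpack cosmetic accessory, giving the pack simple collision for physics.
# It holds a single body on the pack bone, shaped as one rotated, offset capsule.
# There are no constraints, so it defines no jointed or articulated parts.
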PhysicsAsset /BRCosmetics/Accessories/FORT_Backpacks/M_MED_RustLord_Remix_Backpack/M_MED_Rust_Lord_Remix_Backpack_PhysicsAsset
bodies {pack [{shapeType=Capsule, length=18.3975, radius=17.6377, center=<0.5567 -8.6287 -7.0569>, rotation=<rot 0 -91.1227 -80.6918>}]}
constraints []
